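# Two-sided, Nanite-ready foliage material for instanced meshes in the PCG SimpleForest sample.
# Color is the FoliageColor parameter plus a slight variation from world-transformed local origin, computed per vertex through a VertexInterpolator.
# An optional EnableWPO switch pulls vertices toward the instance pivot as camera distance grows, shaped by Dist and Pow, shrinking distant foliage.
# No textures are used, only constants and parameters, including a NormalScale-adjusted constant normal.
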Material /PCG/SampleContent/SimpleForest/Materials/PCG_Foliage_01
graph {
 properties {bUsedWithInstancedStaticMeshes=true, bUsedWithNanite=true, pixelDepthOffsetMode=PDOM_Legacy, twoSided=true, usedShadingModels=Default Lit}
 nodes [
  Constant() {r=0.8}
  Normalize(VectorInput=Multiply(A=Constant3Vector() {constant=<rgba 0.5 0.5 1 0>}, B=VectorParameter() {parameterName=NormalScale, defaultValue=<rgba 1 1 1 1>}.RGB))
  n1 = TransformPosition(None=Constant3Vector()) {transformSourceType=TRANSFORMPOSSOURCE_Instance, transformType=TRANSFORMPOSSOURCE_World}
  n2 = WorldPosition()
  n3 = CameraPositionWS()
  n4 = Saturate(None=Divide(A=Distance(A=@n3, B=@n1), B=ScalarParameter() {defaultValue=1000, parameterName=Dist}) {constB=500})
  n5 = ScalarParameter() {defaultValue=1, parameterName=Pow}
  StaticSwitchParameter(True=Multiply(A=Subtract(A=@n1, B=@n2.XYZ), B=Power(Base=@n4, Exp=@n5) {constExponent=1}) {constB=0}, False=Constant()) {parameterName=EnableWPO}
  n6 = Constant3Vector()
  n7 = Frac(None=Multiply(A=TransformPosition(None=@n6) {transformType=TRANSFORMPOSSOURCE_World}) {constB=0.01})
  VertexInterpolator(VS=Add(A=VectorParameter() {parameterName=FoliageColor, defaultValue=<rgba 0.0072 0.2622 0.0837 0>}.RGB, B=Multiply(A=@n7) {constB=0.1}))
 ]
}
